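# Social panel blade menu widget for the frontend main menu, derived from FortBladeMenu_SocialPanel, hosting the user lists view in a fixed-width, safe-zoned column.
# OnOpened and OnClosed play mirrored short animations that fade, slide, scale and shear the background border.
# An interrupted animation is paused and the opposite one starts from the mirrored time, so reversals stay seamless.
# The widget collapses itself when the close animation finishes.
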WidgetBlueprint /FortniteCoreUI/Frontend/MainMenu/BladeMenu_SocialPanel
animations [
 OnOpen {
  binding_count=1
  length_seconds=0.2
  object=&self:OnOpen
  sequence {bindings [Border_7 {
     id=#1
     possessed_class=Border
     tracks [
      {
       type=MovieSceneFloatTrack
       display_name=Render Opacity
       sections [{type=MovieSceneFloatSection, end_frame=4, start_frame=0, channels [None {type=MovieSceneScriptingFloatChannel, keys [[0,0],[4,1]]}]}]
      }
      {
       type=MovieScene2DTransformTrack
       sections [{
         type=MovieScene2DTransformSection
         channels [
          Translation.X {type=MovieSceneScriptingFloatChannel, keys [[0,-256],[4,0]]}
          Translation.Y {type=MovieSceneScriptingFloatChannel, keys [[0,64],[4,0]]}
          Angle {type=MovieSceneScriptingFloatChannel}
          Scale.X {type=MovieSceneScriptingFloatChannel, keys [[0,0.375],[4,1]]}
          Scale.Y {type=MovieSceneScriptingFloatChannel, keys [[0,0.875],[4,1]]}
          Shear.X {type=MovieSceneScriptingFloatChannel}
          Shear.Y {type=MovieSceneScriptingFloatChannel, keys [[0,-16],[4,0]]}
         ]
        }]
      }
     ]
    }]}
 }
 OnClose {
  binding_count=1
  length_seconds=0.2
  object=&self:OnClose
  sequence {bindings [Border_7 {
     id=#1
     possessed_class=Border
     tracks [
      {
       type=MovieSceneFloatTrack
       display_name=Render Opacity
       sections [{type=MovieSceneFloatSection, end_frame=4, start_frame=0, channels [None {type=MovieSceneScriptingFloatChannel, keys [[0,1],[4,0]]}]}]
      }
      {
       type=MovieScene2DTransformTrack
       sections [{
         type=MovieScene2DTransformSection
         channels [
          Translation.X {type=MovieSceneScriptingFloatChannel, keys [[0,0],[4,-256]]}
          Translation.Y {type=MovieSceneScriptingFloatChannel, keys [[0,0],[4,64]]}
          Angle {type=MovieSceneScriptingFloatChannel}
          Scale.X {type=MovieSceneScriptingFloatChannel, keys [[0,1],[4,0.375]]}
          Scale.Y {type=MovieSceneScriptingFloatChannel, keys [[0,1],[4,0.875]]}
          Shear.X {type=MovieSceneScriptingFloatChannel}
          Shear.Y {type=MovieSceneScriptingFloatChannel, keys [[0,0],[4,-16]]}
         ]
        }]
      }
     ]
    }]}
 }
]
graphs [EventGraph {nodes [
   n1 = OnCloseAnimationFinished() {is=title K2Node_CustomEvent}
   > SetVisibility(InVisibility=Collapsed)
   Event OnInitialized()
   > BindToAnimationFinished(Animation=Get OnClose() -> (OnClose:Widget Animation Object Reference), Delegate=@n1)
   Event OnOpened() {is=title K2Node_Event}
   n2 = Get OnClose()
   > n3 = Branch(Condition=IsAnimationPlaying(InAnimation=@n2))
   > n4 = PauseAnimation!(execute, self:User Widget Object Reference, InAnimation:Widget Animation Object Reference=@n2) -> (then, ReturnValue:float)
   n5 = Get OnOpen() -> (OnOpen:Widget Animation Object Reference)
   > PlayAnimation(InAnimation=@n5, StartAtTime=float - float(A=GetEndTime(self=@n5), B=@n4) {is=title K2Node_CallFunction})
   PlayAnimation(execute=@n3.else, InAnimation=@n5)
   Event OnClosed() {is=title K2Node_Event}
   n6 = Get OnOpen()
   > n7 = Branch(Condition=IsAnimationPlaying(InAnimation=@n6))
   > n8 = PauseAnimation(InAnimation=@n6)
   n9 = Get OnClose()
   > PlayAnimation(InAnimation=@n9, StartAtTime=float - float(A=GetEndTime(self=@n9), B=@n8))
   PlayAnimation(execute=@n7.else, InAnimation=@n9)
  ]}]
variables []
widgets {
 info {inheritedWidgetCount=1, parentClass=/Script/FortniteUI.FortBladeMenu_SocialPanel, rootWidgetClass=&Border, widgetCount=4}
 widgets [
  Border_7 {
   widgetClassPath=&Border
   widget=Border_7
   properties {background {
     bIsDynamicallyLoaded=false
     drawAs=Image
     imageType=NoImage
     mirroring=NoMirror
     resourceName=None
     resourceObject=/FortniteCoreUI/UI/SocialPanel/Art/Materials/M_UI_SocialPanelBackground
     tiling=NoTile
     imageSize=<32 32>
     margin=<ltrb 0 0 0 0>
     outlineSettings {
      bUseBrushTransparency=false
      roundingType=HalfHeightRadius
      width=0
      color {colorUseRule=UseColor_Specified, specifiedColor=<rgba 0 0 0 0>}
      cornerRadii=<0 0 0 0>
     }
     tintColor {colorUseRule=UseColor_Specified, specifiedColor=<rgba 1 1 1 1>}
     uVRegion {bIsValid=false, max=<0 0>, min=<0 0>}
    }}
  }
  SafeZone_Content {
   widgetClassPath=&SafeZone
   bIsVariable=true
   parent=Border_7
   widget=SafeZone_Content
   properties {padBottom=false, padRight=false}
   slot {class=BorderSlot}
  }
  SizeBox_0 {
   widgetClassPath=&SizeBox
   parent=SafeZone_Content
   widget=SizeBox_0
   properties {bOverride_WidthOverride=true, widthOverride=560}
   slot {class=SafeZoneSlot}
  }
  SocialPanel_UserLists {
   widgetClassPath=/FortniteCoreUI/UI/SocialPanel/SocialPanelView_UserLists
   bInherited=true
   bIsVariable=true
   parent=SizeBox_0
   widget=SocialPanel_UserLists
   slot {class=SizeBoxSlot}
  }
 ]
}
