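# Season XP bar widget, built on FortSeasonDataWidget, showing the current level, the XP remaining to the next level, and a progress track.
# Level and progress events call Update Text, which formats both rich text labels and writes CurrentProgress to the track material's Progress_Start parameter.
# Rested and supercharged events set the material's IsSupercharged parameter, and the rested event also toggles the "Supercharged!" badge.
# PreConstruct resizes that badge's text on devices with small-text scaling enabled.
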WidgetBlueprint /FortniteCoreUI/UI/Progression/BladesMenu_XpBar
graphs [
 EventGraph {nodes [
   n1 = Event OnUpdatedRestedState!() -> (OutputDelegate, then, bIsRested:bool) {is=title K2Node_Event}
   > n2 = GetDynamicMaterial(self=Get Image_RaceTrack() -> (Image_RaceTrack:Image Object Reference))
   > SetScalarParameterValue(self=@n2, ParameterName=IsSupercharged, Value=SelectFloat(A=1, B=0, bPickA=@n1.bIsRested))
   > SetVisibility(
    self=Get RestedXPTextBox() -> (RestedXPTextBox:Border Object Reference)
    InVisibility=Select(Option 0=Collapsed, Option 1=SelfHitTestInvisible, Index=@n1.bIsRested)
   )
   Event PreConstruct()
   > Branch(Condition=GetSmallTextScalingMultiplier!() -> (ScalingMultiplier:float, IsEnabledOnDevice:bool).IsEnabledOnDevice)
   > SetFontSize!(execute, self:Text Object Reference=Get Text_Supercharged() -> (Text_Supercharged:Common Text Object Reference), DisplayFontSize:float=16) -> (then)
   Event OnSuperCharged() {is=title K2Node_Event}
   > n3 = GetDynamicMaterial(self=Get Image_RaceTrack())
   > SetScalarParameterValue(self=@n3, ParameterName=IsSupercharged, Value=1)
   Event OnLevelChanged() {is=title K2Node_Event}
   > Update Text()
   Event OnProgressUpdated() {is=title K2Node_Event}
   > Update Text()
  ]}
 Update Text {nodes [
   Update Text() {is=title K2Node_FunctionEntry}
   n1 = Get XpToNextLevel() -> (XpToNextLevel:int=0)
   n2 = Get CurrentXP() -> (CurrentXP:int=0)
   n3 = Get CurrentLevel() -> (CurrentLevel:int=0)
   > SetText(
    self=Get RichText_Level() -> (RichText_Level:Common Rich Text Block Object Reference)
    InText=Format Text!(
     Format="LOCTEXT(\"<text color=\\\"FFF\\\">{XPamount} XP</> to LVL <text size=\\\"17\\\">{level}</>\")"
     XPamount:int=int - int(A=@n1, B=@n2) {is=title K2Node_CallFunction}
     level:int=int + int(A=@n3, B=1) -> (ReturnValue=0) {is=title K2Node_CommutativeAssociativeBinaryOperator}
    ) -> (Result)
   )
   > SetText(
    self=Get RichText_CurrentLevel() -> (RichText_CurrentLevel:Common Rich Text Block Object Reference)
    InText=Format Text!(Format="LOCTEXT(\"LVL <text color=\\\"FFF\\\" size=\\\"17\\\">{level}</>\")", level=@n3) -> (Result)
   )
   > n4 = GetDynamicMaterial(self=Get Image_RaceTrack())
   > SetScalarParameterValue(self=@n4, ParameterName=Progress_Start, Value=Get CurrentProgress() -> (CurrentProgress:float))
  ]}
]
variables [Max BPLevel for Rewards {type=integer, default=200}]
widgets {
 info {inheritedWidgetCount=2, parentClass=/Script/FortniteUI.FortSeasonDataWidget, rootWidgetClass=&SizeBox, widgetCount=23}
 widgets [
  SizeBox_0 {
   widgetClassPath=&SizeBox
   widget=SizeBox_0
   properties {bOverride_HeightOverride=true, heightOverride=72, visibility=HitTestInvisible}
  }
  Overlay_2 {widgetClassPath=&Overlay, parent=SizeBox_0, widget=Overlay_2, slot {class=SizeBoxSlot}}
  Backing_Darken {
   widgetClassPath=&Image
   bIsVariable=true
   parent=Overlay_2
   widget=Backing_Darken
   properties {
    renderOpacity=0
    visibility=SelfHitTestInvisible
    brush {
     bIsDynamicallyLoaded=false
     drawAs=Box
     imageType=NoImage
     mirroring=NoMirror
     resourceName=None
     resourceObject=/FortUILibrary/Foundation/Art_ToolKit/Tools/M_UI_XP_Everywhere_Backing_Darken
     tiling=NoTile
     imageSize=<16 16>
     margin=<ltrb 0.5 0.5 0.5 0.5>
     outlineSettings {
      bUseBrushTransparency=false
      roundingType=HalfHeightRadius
      width=0
      color {colorUseRule=UseColor_Specified, specifiedColor=<rgba 0 0 0 0>}
      cornerRadii=<0 0 0 0>
     }
     tintColor {colorUseRule=UseColor_Specified, specifiedColor=<rgba 1 1 1 1>}
     uVRegion {bIsValid=false, max=<0 0>, min=<0 0>}
    }
    colorAndOpacity=<rgba 1 1 1 0.96>
   }
   slot {class=OverlaySlot, properties {horizontalAlignment=HAlign_Fill, verticalAlignment=VAlign_Fill}}
  }
  Backing_Color {
   widgetClassPath=&Image
   bIsVariable=true
   parent=Overlay_2
   widget=Backing_Color
   properties {
    visibility=SelfHitTestInvisible
    brush {
     bIsDynamicallyLoaded=false
     drawAs=Box
     imageType=FullColor
     mirroring=NoMirror
     resourceName=None
     resourceObject=/FortniteCoreUI/Athena/HUD/Quests/Art/M_Ui_Quest_XPNotification_DarkPurple
     tiling=NoTile
     imageSize=<64 64>
     margin=<ltrb 0.3 0.3 0.3 0.3>
     outlineSettings {
      bUseBrushTransparency=false
      roundingType=HalfHeightRadius
      width=0
      color {colorUseRule=UseColor_Specified, specifiedColor=<rgba 0 0 0 0>}
      cornerRadii=<0 0 0 0>
     }
     tintColor {colorUseRule=UseColor_Specified, specifiedColor=<rgba 1 1 1 1>}
     uVRegion {bIsValid=false, max=<0 0>, min=<0 0>}
    }
   }
   slot {class=OverlaySlot, properties {horizontalAlignment=HAlign_Fill, verticalAlignment=VAlign_Fill}}
  }
  Backing_Outline {
   widgetClassPath=&Image
   bIsVariable=true
   parent=Overlay_2
   widget=Backing_Outline
   properties {
    renderOpacity=0
    visibility=SelfHitTestInvisible
    brush {
     bIsDynamicallyLoaded=false
     drawAs=Box
     imageType=NoImage
     mirroring=NoMirror
     resourceName=None
     resourceObject=/FortUILibrary/Foundation/Art_ToolKit/Tools/M_UI_XP_Everywhere_Backing_Color
     tiling=NoTile
     imageSize=<16 16>
     margin=<ltrb 0.5 0.5 0.5 0.5>
     outlineSettings {
      bUseBrushTransparency=false
      roundingType=HalfHeightRadius
      width=0
      color {colorUseRule=UseColor_Specified, specifiedColor=<rgba 0 0 0 0>}
      cornerRadii=<0 0 0 0>
     }
     tintColor {colorUseRule=UseColor_Specified, specifiedColor=<rgba 1 1 1 1>}
     uVRegion {bIsValid=false, max=<0 0>, min=<0 0>}
    }
    renderTransform {angle=0, scale=<1.005 1.025>, shear=<0 0>, translation=<0 0>}
   }
   slot {class=OverlaySlot, properties {horizontalAlignment=HAlign_Fill, verticalAlignment=VAlign_Fill}}
  }
  HorizontalBox_1 {
   widgetClassPath=&HorizontalBox
   parent=Overlay_2
   widget=HorizontalBox_1
   slot {class=OverlaySlot, properties {verticalAlignment=VAlign_Center, padding=<ltrb 8 4 4 4>}}
  }
  SocialAvatarIcon {
   widgetClassPath=/Script/FortniteUI.FortSocialAvatarIcon
   bInherited=true
   bIsVariable=true
   parent=HorizontalBox_1
   widget=SocialAvatarIcon
   properties {
    avatarMaterial=/Game/UI/SocialPanel/Art/Materials/UI_UserPortrait_Buffet_XP
    visibility=Collapsed
    brush {
     bIsDynamicallyLoaded=false
     drawAs=Image
     imageType=NoImage
     mirroring=NoMirror
     resourceName=None
     resourceObject=&UI_UserPortrait_Buffet_XP
     tiling=NoTile
     imageSize=<56 56>
     margin=<ltrb 0 0 0 0>
     outlineSettings {
      bUseBrushTransparency=false
      roundingType=HalfHeightRadius
      width=0
      color {colorUseRule=UseColor_Specified, specifiedColor=<rgba 0 0 0 0>}
      cornerRadii=<0 0 0 0>
     }
     tintColor {colorUseRule=UseColor_Specified, specifiedColor=<rgba 1 1 1 1>}
     uVRegion {bIsValid=false, max=<0 0>, min=<0 0>}
    }
   }
   slot {class=HorizontalBoxSlot}
  }
  SizeBox_1 {
   widgetClassPath=&SizeBox
   parent=HorizontalBox_1
   widget=SizeBox_1
   properties {bOverride_HeightOverride=true, bOverride_WidthOverride=true, heightOverride=45, widthOverride=416}
   slot {class=HorizontalBoxSlot, properties {padding=<ltrb 4 0 8 0>, size {sizeRule=Fill, value=1}}}
  }
  VerticalBox_22 {
   widgetClassPath=&VerticalBox
   parent=SizeBox_1
   widget=VerticalBox_22
   slot {class=SizeBoxSlot, properties {verticalAlignment=VAlign_Center, padding=<ltrb 0 0 0 4>}}
  }
  HorizontalBox_0 {
   widgetClassPath=&HorizontalBox
   parent=VerticalBox_22
   widget=HorizontalBox_0
   slot {class=VerticalBoxSlot, properties {padding=<ltrb 2 0 0 8>}}
  }
  RichText_CurrentLevel {
   widgetClassPath=&CommonRichTextBlock
   bIsVariable=true
   parent=HorizontalBox_0
   widget=RichText_CurrentLevel
   properties {
    bOverrideDefaultStyle=true
    lineHeightPercentage=0.8
    mobileTextBlockScale=1.1
    text="LVL <text color=\"FFF\" size=\"17\">X</>"
    textTransformPolicy=ToUpper
    defaultTextStyleOverride {
     overflowPolicy=Clip
     transformPolicy=None
     colorAndOpacity {colorUseRule=UseColor_Specified, specifiedColor=<rgba 0.0875 0.1006 1 1>}
     font {
      bForceMonospaced=false
      bMaterialIsStencil=false
      fontMaterial=null
      fontObject=&HeadingNow
      letterSpacing=0
      monospacedWidth=1
      size=12
      skewAmount=0
      typefaceFontName=86 Bold
      outlineSettings {
       bApplyOutlineToDropShadows=false
       bMiteredCorners=false
       bSeparateFillAlpha=false
       outlineMaterial=null
       outlineSize=0
       outlineColor=<rgba 0 0 0 1>
      }
     }
     highlightColor {colorUseRule=UseColor_Specified, specifiedColor=<rgba 0 0 0 1>}
     highlightShape {
      bIsDynamicallyLoaded=false
      drawAs=Image
      imageType=NoImage
      mirroring=NoMirror
      resourceName=None
      resourceObject=null
      tiling=NoTile
      imageSize=<32 32>
      margin=<ltrb 0 0 0 0>
      outlineSettings {
       bUseBrushTransparency=false
       roundingType=HalfHeightRadius
       width=0
       color {colorUseRule=UseColor_Specified, specifiedColor=<rgba 0 0 0 0>}
       cornerRadii=<0 0 0 0>
      }
      tintColor {colorUseRule=UseColor_Specified, specifiedColor=<rgba 1 1 1 1>}
      uVRegion {bIsValid=false, max=<0 0>, min=<0 0>}
     }
     selectedBackgroundColor {colorUseRule=UseColor_Foreground, specifiedColor=<rgba 1 0 1 1>}
     shadowColorAndOpacity=<rgba 0 0 0 1>
     shadowOffset=<0 0>
     strikeBrush {
      bIsDynamicallyLoaded=false
      drawAs=Image
      imageType=NoImage
      mirroring=NoMirror
      resourceName=None
      resourceObject=null
      tiling=NoTile
      imageSize=<32 32>
      margin=<ltrb 0 0 0 0>
      outlineSettings {
       bUseBrushTransparency=false
       roundingType=HalfHeightRadius
       width=0
       color {colorUseRule=UseColor_Specified, specifiedColor=<rgba 0 0 0 0>}
       cornerRadii=<0 0 0 0>
      }
      tintColor {colorUseRule=UseColor_Specified, specifiedColor=<rgba 1 1 1 1>}
      uVRegion {bIsValid=false, max=<0 0>, min=<0 0>}
     }
     underlineBrush {
      bIsDynamicallyLoaded=false
      drawAs=Image
      imageType=NoImage
      mirroring=NoMirror
      resourceName=None
      resourceObject=null
      tiling=NoTile
      imageSize=<32 32>
      margin=<ltrb 0 0 0 0>
      outlineSettings {
       bUseBrushTransparency=false
       roundingType=HalfHeightRadius
       width=0
       color {colorUseRule=UseColor_Specified, specifiedColor=<rgba 0 0 0 0>}
       cornerRadii=<0 0 0 0>
      }
      tintColor {colorUseRule=UseColor_Specified, specifiedColor=<rgba 1 1 1 1>}
      uVRegion {bIsValid=false, max=<0 0>, min=<0 0>}
     }
    }
    renderTransform {angle=0, scale=<1 1>, shear=<-12 0>, translation=<0 0>}
   }
   slot {class=HorizontalBoxSlot, properties {padding=<ltrb 3 0 16 0>, size {sizeRule=Fill, value=1}}}
  }
  RichText_Level {
   widgetClassPath=&CommonRichTextBlock
   bIsVariable=true
   parent=HorizontalBox_0
   widget=RichText_Level
   properties {
    bOverrideDefaultStyle=true
    lineHeightPercentage=0.8
    mobileTextBlockScale=1.2
    text="<text color=\"FFF\">0 XP</> to LVL <text size=\"17\">0</>"
    defaultTextStyleOverride {
     overflowPolicy=Clip
     transformPolicy=None
     colorAndOpacity {colorUseRule=UseColor_Specified, specifiedColor=<rgba 0.0875 0.1006 1 1>}
     font {
      bForceMonospaced=false
      bMaterialIsStencil=false
      fontMaterial=null
      fontObject=&HeadingNow
      letterSpacing=0
      monospacedWidth=1
      size=12
      skewAmount=0
      typefaceFontName=86 Bold
      outlineSettings {
       bApplyOutlineToDropShadows=false
       bMiteredCorners=false
       bSeparateFillAlpha=false
       outlineMaterial=null
       outlineSize=0
       outlineColor=<rgba 0 0 0 1>
      }
     }
     highlightColor {colorUseRule=UseColor_Specified, specifiedColor=<rgba 0 0 0 1>}
     highlightShape {
      bIsDynamicallyLoaded=false
      drawAs=Image
      imageType=NoImage
      mirroring=NoMirror
      resourceName=None
      resourceObject=null
      tiling=NoTile
      imageSize=<32 32>
      margin=<ltrb 0 0 0 0>
      outlineSettings {
       bUseBrushTransparency=false
       roundingType=HalfHeightRadius
       width=0
       color {colorUseRule=UseColor_Specified, specifiedColor=<rgba 0 0 0 0>}
       cornerRadii=<0 0 0 0>
      }
      tintColor {colorUseRule=UseColor_Specified, specifiedColor=<rgba 1 1 1 1>}
      uVRegion {bIsValid=false, max=<0 0>, min=<0 0>}
     }
     selectedBackgroundColor {colorUseRule=UseColor_Foreground, specifiedColor=<rgba 1 0 1 1>}
     shadowColorAndOpacity=<rgba 0 0 0 1>
     shadowOffset=<0 0>
     strikeBrush {
      bIsDynamicallyLoaded=false
      drawAs=Image
      imageType=NoImage
      mirroring=NoMirror
      resourceName=None
      resourceObject=null
      tiling=NoTile
      imageSize=<32 32>
      margin=<ltrb 0 0 0 0>
      outlineSettings {
       bUseBrushTransparency=false
       roundingType=HalfHeightRadius
       width=0
       color {colorUseRule=UseColor_Specified, specifiedColor=<rgba 0 0 0 0>}
       cornerRadii=<0 0 0 0>
      }
      tintColor {colorUseRule=UseColor_Specified, specifiedColor=<rgba 1 1 1 1>}
      uVRegion {bIsValid=false, max=<0 0>, min=<0 0>}
     }
     underlineBrush {
      bIsDynamicallyLoaded=false
      drawAs=Image
      imageType=NoImage
      mirroring=NoMirror
      resourceName=None
      resourceObject=null
      tiling=NoTile
      imageSize=<32 32>
      margin=<ltrb 0 0 0 0>
      outlineSettings {
       bUseBrushTransparency=false
       roundingType=HalfHeightRadius
       width=0
       color {colorUseRule=UseColor_Specified, specifiedColor=<rgba 0 0 0 0>}
       cornerRadii=<0 0 0 0>
      }
      tintColor {colorUseRule=UseColor_Specified, specifiedColor=<rgba 1 1 1 1>}
      uVRegion {bIsValid=false, max=<0 0>, min=<0 0>}
     }
    }
    renderTransform {angle=0, scale=<1 1>, shear=<-12 0>, translation=<0 0>}
   }
   slot {class=HorizontalBoxSlot, properties {horizontalAlignment=HAlign_Right, verticalAlignment=VAlign_Bottom}}
  }
  Image_RaceTrack {
   widgetClassPath=&Image
   bIsVariable=true
   parent=VerticalBox_22
   widget=Image_RaceTrack
   properties {
    visibility=SelfHitTestInvisible
    brush {
     bIsDynamicallyLoaded=false
     drawAs=Image
     imageType=NoImage
     mirroring=NoMirror
     resourceName=None
     resourceObject=/FortUILibrary/Foundation/Art_ToolKit/ProgressBars/M_UI_XP_Everywhere_ProgressBar
     tiling=NoTile
     imageSize=<32 10>
     margin=<ltrb 0.5 0 0.5 0>
     outlineSettings {
      bUseBrushTransparency=false
      roundingType=HalfHeightRadius
      width=0
      color {colorUseRule=UseColor_Specified, specifiedColor=<rgba 0 0 0 0>}
      cornerRadii=<0 0 0 0>
     }
     tintColor {colorUseRule=UseColor_Specified, specifiedColor=<rgba 1 1 1 1>}
     uVRegion {bIsValid=false, max=<0 0>, min=<0 0>}
    }
    renderTransform {angle=0, scale=<1 1>, shear=<-8 0>, translation=<0 0>}
   }
   slot {class=VerticalBoxSlot}
  }
  ScaleBox_172 {
   widgetClassPath=&ScaleBox
   parent=VerticalBox_22
   widget=ScaleBox_172
   properties {visibility=Collapsed}
   slot {class=VerticalBoxSlot}
  }
  LevelRewards {
   widgetClassPath=/FortniteCoreUI/HUD/MiniMap/AthenaChallengeRewards
   bInherited=true
   bIsVariable=true
   parent=ScaleBox_172
   widget=LevelRewards
   slot {class=ScaleBoxSlot, properties {horizontalAlignment=HAlign_Fill, verticalAlignment=VAlign_Fill}}
  }
  RestedXPTextBox {
   widgetClassPath=&Border
   bIsVariable=true
   parent=Overlay_2
   widget=RestedXPTextBox
   properties {
    visibility=Collapsed
    background {
     bIsDynamicallyLoaded=false
     drawAs=NoDrawType
     imageType=NoImage
     mirroring=NoMirror
     resourceName=None
     resourceObject=null
     tiling=NoTile
     imageSize=<32 32>
     margin=<ltrb 0 0 0 0>
     outlineSettings {
      bUseBrushTransparency=false
      roundingType=HalfHeightRadius
      width=0
      color {colorUseRule=UseColor_Specified, specifiedColor=<rgba 0 0 0 0>}
      cornerRadii=<0 0 0 0>
     }
     tintColor {colorUseRule=UseColor_Specified, specifiedColor=<rgba 1 1 1 1>}
     uVRegion {bIsValid=false, max=<0 0>, min=<0 0>}
    }
    padding=<ltrb 0 0 0 0>
    renderTransform {angle=0, scale=<1 1>, shear=<0 0>, translation=<0 -38>}
   }
   slot {class=OverlaySlot, properties {padding=<ltrb 22 0 0 0>}}
  }
  Overlay_4 {widgetClassPath=&Overlay, parent=RestedXPTextBox, widget=Overlay_4, slot {class=BorderSlot, properties {padding=<ltrb 0 0 0 0>}}}
  Image_355 {
   widgetClassPath=&Image
   bIsVariable=true
   parent=Overlay_4
   widget=Image_355
   properties {
    visibility=SelfHitTestInvisible
    brush {
     bIsDynamicallyLoaded=false
     drawAs=Box
     imageType=NoImage
     mirroring=NoMirror
     resourceName=None
     resourceObject=/Game/Athena/HUD/LTM/Wax/Materials/M_UI_TextGlow_Add_Gradient_Purple_XP
     tiling=NoTile
     imageSize=<48 48>
     margin=<ltrb 0.5 0.5 0.5 0.5>
     outlineSettings {
      bUseBrushTransparency=false
      roundingType=HalfHeightRadius
      width=0
      color {colorUseRule=UseColor_Specified, specifiedColor=<rgba 0 0 0 0>}
      cornerRadii=<0 0 0 0>
     }
     tintColor {colorUseRule=UseColor_Specified, specifiedColor=<rgba 1 1 1 1>}
     uVRegion {bIsValid=false, max=<0 0>, min=<0 0>}
    }
    colorAndOpacity=<rgba 1 1 1 0.0625>
    renderTransform {angle=0, scale=<1.25 1.333>, shear=<0 0>, translation=<0 0>}
   }
   slot {class=OverlaySlot, properties {horizontalAlignment=HAlign_Fill, verticalAlignment=VAlign_Fill}}
  }
  HorizontalBox {
   widgetClassPath=&HorizontalBox
   parent=Overlay_4
   widget=HorizontalBox
   slot {class=OverlaySlot, properties {verticalAlignment=VAlign_Center}}
  }
  Overlay_3 {
   widgetClassPath=&Overlay
   parent=HorizontalBox
   widget=Overlay_3
   properties {renderTransform {angle=0, scale=<1 1>, shear=<0 0>, translation=<0 1>}}
   slot {class=HorizontalBoxSlot}
  }
  Image_198 {
   widgetClassPath=&Image
   bIsVariable=true
   parent=Overlay_3
   widget=Image_198
   properties {
    visibility=SelfHitTestInvisible
    brush {
     bIsDynamicallyLoaded=false
     drawAs=Image
     imageType=NoImage
     mirroring=NoMirror
     resourceName=None
     resourceObject=&M_UI_TextGlow_Add_Gradient_Purple_XP
     tiling=NoTile
     imageSize=<32 32>
     margin=<ltrb 0 0 0 0>
     outlineSettings {
      bUseBrushTransparency=false
      roundingType=HalfHeightRadius
      width=0
      color {colorUseRule=UseColor_Specified, specifiedColor=<rgba 0 0 0 0>}
      cornerRadii=<0 0 0 0>
     }
     tintColor {colorUseRule=UseColor_Specified, specifiedColor=<rgba 1 1 1 1>}
     uVRegion {bIsValid=false, max=<0 0>, min=<0 0>}
    }
    colorAndOpacity=<rgba 1 1 1 0.125>
    renderTransform {angle=0, scale=<2 2>, shear=<0 0>, translation=<0 0>}
   }
   slot {class=OverlaySlot, properties {horizontalAlignment=HAlign_Center, verticalAlignment=VAlign_Center}}
  }
  Image_108 {
   widgetClassPath=&Image
   bIsVariable=true
   parent=Overlay_3
   widget=Image_108
   properties {
    brush {
     bIsDynamicallyLoaded=false
     drawAs=Image
     imageType=FullColor
     mirroring=NoMirror
     resourceName=None
     resourceObject=/FortniteCoreUI/UI/Foundation/Textures/Icons/Items/UniversalXP/T_UI_UniversalXP_Icon_LevelUp_Standard_64
     tiling=NoTile
     imageSize=<32 32>
     margin=<ltrb 0 0 0 0>
     outlineSettings {
      bUseBrushTransparency=false
      roundingType=HalfHeightRadius
      width=0
      color {colorUseRule=UseColor_Specified, specifiedColor=<rgba 0 0 0 0>}
      cornerRadii=<0 0 0 0>
     }
     tintColor {colorUseRule=UseColor_Specified, specifiedColor=<rgba 1 1 1 1>}
     uVRegion {bIsValid=false, max=<0 0>, min=<0 0>}
    }
    renderTransform {angle=0, scale=<1 1>, shear=<0 0>, translation=<0 1>}
   }
   slot {class=OverlaySlot}
  }
  Text_Supercharged {
   widgetClassPath=&CommonTextBlock
   bIsVariable=true
   parent=HorizontalBox
   widget=Text_Supercharged
   properties {
    text="Supercharged!"
    textTransformPolicy=ToUpper
    font {
     bForceMonospaced=false
     bMaterialIsStencil=false
     fontMaterial=null
     fontObject=&HeadingNow
     letterSpacing=0
     monospacedWidth=1
     size=9
     skewAmount=0
     typefaceFontName=86 Bold
     outlineSettings {
      bApplyOutlineToDropShadows=false
      bMiteredCorners=false
      bSeparateFillAlpha=false
      outlineMaterial=/Game/UI/Foundation/Materials/M-FontShadow
      outlineSize=2
      outlineColor=<rgba 0.0083 0 0.025 1>
     }
    }
    renderTransform {angle=0, scale=<1 1>, shear=<-12 0>, translation=<-5 1>}
    shadowOffset=<0 0>
   }
   slot {class=HorizontalBoxSlot, properties {verticalAlignment=VAlign_Center, padding=<ltrb 4 2 0 0>}}
  }
 ]
}
