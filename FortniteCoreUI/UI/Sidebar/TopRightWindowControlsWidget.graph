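# Top-right window controls widget, a FortBladeMenu subclass that shows the shared windowchromebuttons widget.
# A safe zone, with left padding off, wraps a legacy CommonVisibilityWidget that hides the buttons on consoles, mobile and Mac.
# The buttons default to Visible.
# The event graph is unwired: OnRefreshOptions drives nothing, and neither the SetVisibility call nor the IsInZone-based visibility Select is connected.
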WidgetBlueprint /FortniteCoreUI/UI/Sidebar/TopRightWindowControlsWidget
graphs [EventGraph {nodes [
   SetVisibility(self=Get windowchromebuttons() -> (windowchromebuttons:Windowchromebuttons Object Reference), InVisibility=Visible)
   Event OnRefreshOptions() {is=title K2Node_Event}
   Select(
    Option 0=Collapsed
    Option 1=Visible
    Index=IsInZone!(self:Fort Global UIContext Object Reference=Get FortGlobalUIContext!() -> (ReturnValue:Fort Global UIContext Object Reference) {is=title K2Node_GetBlueprintContextEx}) -> (ReturnValue:bool)
   )
  ]}]
variables []
widgets {
 info {inheritedWidgetCount=0, parentClass=/Script/FortniteUI.FortBladeMenu, rootWidgetClass=&SafeZone, widgetCount=3}
 widgets [
  SafeZone_Content {widgetClassPath=&SafeZone, bIsVariable=true, widget=SafeZone_Content, properties {padLeft=false}}
  CommonVisibilityWidget {
   widgetClassPath=&CommonVisibilityWidgetLegacy
   parent=SafeZone_Content
   widget=CommonVisibilityWidget
   properties {
    bShowForAndroid=false
    bShowForErebus=false
    bShowForIOS=false
    bShowForMac=false
    bShowForPS4=false
    bShowForPS5=false
    bShowForSwitch2=false
    bShowForXBox=false
    bShowForXSX=false
    padding=<ltrb 0 0 0 0>
   }
   slot {class=SafeZoneSlot}
  }
  windowchromebuttons {
   widgetClassPath=/Game/UI/Foundation/Shell/windowchromebuttons
   bIsVariable=true
   parent=CommonVisibilityWidget
   widget=windowchromebuttons
   properties {visibility=Visible}
   slot {class=BorderSlot, properties {padding=<ltrb 0 0 0 0>}}
  }
 ]
}
