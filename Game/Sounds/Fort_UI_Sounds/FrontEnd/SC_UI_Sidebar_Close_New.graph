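# UI sound cue for closing a front-end sidebar.
# It plays a single sidebar-close wave through a short fixed delay, at reduced volume in the Fort_UI sound class.
# A concurrency override allows one instance at a time and stops the oldest, and virtualization is disabled.
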SoundCue /Game/Sounds/Fort_UI_Sounds/FrontEnd/SC_UI_Sidebar_Close_New
root=Delay_1
nodes [
 {id=Delay_1, children [WavePlayer_3], properties {delayMax=0.04, delayMin=0.04}}
 {id=WavePlayer_3, properties {soundWaveAssetPtr=/Game/Sounds/Fort_UI_Sounds/FrontEnd/Social/UI_Sidebar_Close_Right_05}}
]
properties {
 bOverrideConcurrency=true
 maxDistance=2097152
 soundClassObject=&Fort_UI
 virtualizationMode=Disabled
 volumeMultiplier=0.5
 concurrencyOverrides {maxCount=1, resolutionRule=StopOldest}
}
